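# Niagara smoke effect from the FX reuse examples, with one sprite emitter; its name points to fake light and shadow, though no lighting logic appears here.
# Burst and rate spawning place fixed-size, slowly rotating SubUV sprites in an offset sphere, using InheritSourceMovement for position.
# Dark blue-gray color fades in and out over each particle's life.
# FN_PickFloatDependOnPlatform varies sphere radius, height and opacity by platform quality level.
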NiagaraSystem /Game/NeverCook/FX/Niagara_FX_Reuse_Examples/Smoke/NS_Smoke_FakeLightAndShadow
dependencies {
 usedDataInterfaces [
  &NiagaraDataInterfacePlatformSet
  &NiagaraDataInterfaceSpriteRendererInfo
  &NiagaraDataInterfaceCurve
 ]
 usedDynamicInputs []
 usedModules [&SystemState]
}
emitters {Smoke001 {
  emitterUpdateScript [
   &EmitterState(Life Cycle Mode=::NewEnumerator1, Loop Behavior=::NewEnumerator1, Loop Delay=1.5, UseLoopDelay=false)
   &SpawnBurst_Instantaneous(Spawn Count=10, Use Spawn Probability=false, UseLoopCountLimit=false)
   &SpawnRate(SpawnRate=10, Use Spawn Probability=false)
  ]
  particleSpawnScript [
   &InitializeParticle(
    Lifetime Max=3
    Lifetime Min=3
    Lifetime Mode=::NewEnumerator1
    Sprite Rotation Mode=::NewEnumerator2
    Sprite Size Mode=::NewEnumerator1
    Uniform Sprite Size Max=150
    Uniform Sprite Size Min=150
    UsePositionOffset=false
    Write Material Random=true
   )
   &ShapeLocation(
    MaskSpawnGroup=false
    Offset=&MakeVector(
     X=0
     Y=0
     Z=&FN_PickFloatDependOnPlatform(
      TargetPlatform=DataInterface() {propertyValues {Platforms {
         qualityLevelMask=-4
         deviceProfileStates [
          {profileName=Windows, qualityLevelMask=2, setQualityLevelMask=2}
          {profileName=Mac, qualityLevelMask=2, setQualityLevelMask=2}
          {profileName=Linux, qualityLevelMask=2, setQualityLevelMask=2}
         ]
        }}}
      TargetPlatformActive=80
      TargetPlatformNoActive=120
     )
    )
    Offset Coordinate Space=::Local
    Offset Mode=::NewEnumerator0
    Override Seed=false
    Sphere Radius=&FN_PickFloatDependOnPlatform(TargetPlatform=DataInterface() {propertyValues {Platforms {qualityLevelMask=-2}}}, TargetPlatformActive=120, TargetPlatformNoActive=90)
    Sphere Surface Distribution=1
    UseOwnerScale=true
   )
  ]
  particleUpdateScript [
   &ParticleState()
   &SpriteRotationRate(Rotation Rate=&RandomRangeFloat(Minimum=-20, Maximum=20))
   &InheritSourceMovement(ApplyForce=false, ApplyPosition=true, ApplyPositionFalloff=false)
   &SolveForcesAndVelocity(Clamp Velocity=false, Limit Acceleration=false)
   &SubUVAnimation(
    "Sprite Renderer "=DataInterface() {propertyValues {SpriteRenderer=&self:Smoke001_0.NiagaraSpriteRendererProperties_0}}
    UseEndFrame=false
    UseStartFrame=false
   )
   &ScaleColor(
    Scale Alpha=&FloatFromCurve(
     FloatCurve=DataInterface() {propertyValues {Curve {
        defaultValue=1
        keys [
         {interpMode=RCIM_Cubic, tangentMode=RCTM_User, time=0, value=0}
         {arriveTangent=2.4981, interpMode=RCIM_Cubic, leaveTangent=2.4981, time=0.2, value=1}
         <at 0.4003 1 RCIM_Cubic>
         {interpMode=RCIM_Cubic, tangentMode=RCTM_User, time=0.7654, value=1}
         {tangentMode=RCTM_User, time=1, value=0}
        ]
       }}}
     Scale Curve=&Multiply_Float(
      A=&Multiply_Float(A=0.7, B=&Lerp_Float(A=1.5, B=1.5, Alpha=$Particles.RegionMask.Float))
      B=&FN_PickFloatDependOnPlatform(
       TargetPlatform=DataInterface() {propertyValues {Platforms {
          qualityLevelMask=-4
          deviceProfileStates [
           {profileName=Windows, qualityLevelMask=3, setQualityLevelMask=3}
           {profileName=Mac, qualityLevelMask=2, setQualityLevelMask=2}
           {profileName=Linux, qualityLevelMask=2, setQualityLevelMask=2}
          ]
         }}}
       TargetPlatformActive=1.1
       TargetPlatformNoActive=0.5
      )
     )
    )
    Scale RGB=&MakeVectorFromLinearColorRGB(Color=<rgba 0.0594 0.0847 0.1 1>)
    ScaleA=true
    ScaleRGB=true
   )
  ]
  rendererClasses [&NiagaraSpriteRendererProperties]
  renderers [{rendererClass=&NiagaraSpriteRendererProperties}]
 }}
summary {
 systemName=NS_Smoke_FakeLightAndShadow
 emitters [{emitterName=Smoke001, rendererClasses [&NiagaraSpriteRendererProperties]}]
 userVariables []
}
